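# Physics asset for the Giggle medium male backpack cosmetic, providing simulation bodies for its leather pieces and straps.
# Small capsules sit on the pack bone, left and right leather bones, and two strap bones with end bones.
# Both leather pieces and both strap bases have angularly locked constraints to the pack.
# Each strap end chains to its base with unlocked angular motion and small limit angles, so secondary motion is confined to the strap tips.
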PhysicsAsset /BRCosmetics/Accessories/FORT_Backpacks/Backpack_M_MED_Giggle/Meshes/M_MED_Giggle_Pack_Physics1
bodies {
 dyn_leather_l [{shapeType=Capsule, length=0.5, radius=3, shapeName=dyn_leather_l_capsule, center=<0 0 -1>}]
 dyn_leather_r [{shapeType=Capsule, length=0.2, radius=3, shapeName=dyn_leather_r_capsule, center=<-0.0001 0 -1.0733>}]
 dyn_straps01 [{shapeType=Capsule, length=0.2, radius=2.25, shapeName=dyn_straps01_capsule}]
 dyn_straps01_end [{shapeType=Capsule, length=0.5, radius=1.5, shapeName=dyn_straps01_end_capsule, center=<0.5 0 0>, rotation=<rot -90 0 0>}]
 dyn_straps02 [{shapeType=Capsule, length=0.2, radius=2.25, shapeName=dyn_straps02_capsule}]
 dyn_straps02_end [{shapeType=Capsule, length=0.5, radius=1.25, shapeName=dyn_straps02_end_capsule, center=<0.5 0 0>, rotation=<rot -90 0 0>}]
 pack [{shapeType=Capsule, length=0.2, radius=1, shapeName=pack_capsule}]
}
constraints [
 {
  bone1Name=dyn_straps02
  bone2Name=pack
  swing1LimitDegrees=0
  swing1Motion=Locked
  swing2LimitDegrees=0
  swing2Motion=Locked
  twistLimitDegrees=0
  twistMotion=Locked
 }
 {bone1Name=dyn_straps02_end, bone2Name=dyn_straps02, swing1LimitDegrees=0, swing2LimitDegrees=2, twistLimitDegrees=3}
 {
  bone1Name=dyn_straps01
  bone2Name=pack
  swing1LimitDegrees=45
  swing1Motion=Locked
  swing2LimitDegrees=45
  swing2Motion=Locked
  twistLimitDegrees=45
  twistMotion=Locked
 }
 {
  bone1Name=dyn_leather_l
  bone2Name=pack
  swing1LimitDegrees=5
  swing1Motion=Locked
  swing2LimitDegrees=10
  swing2Motion=Locked
  twistLimitDegrees=45
  twistMotion=Locked
 }
 {
  bone1Name=dyn_leather_r
  bone2Name=pack
  swing1LimitDegrees=5
  swing1Motion=Locked
  swing2LimitDegrees=10
  swing2Motion=Locked
  twistLimitDegrees=45
  twistMotion=Locked
 }
 {bone1Name=dyn_straps01_end, bone2Name=dyn_straps01, swing1LimitDegrees=0, swing2LimitDegrees=3, twistLimitDegrees=1}
]
