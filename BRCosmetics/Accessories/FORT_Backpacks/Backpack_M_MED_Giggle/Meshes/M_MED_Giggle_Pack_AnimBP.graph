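# Animation Blueprint for the Giggle backpack cosmetic, posing the pack and adding physics-driven secondary motion.
# It starts from the reference pose and additively offsets and rolls the pack bone, then adds more roll while bIsSkyDiving is true.
# A RigidBody node then simulates on an override physics asset with custom gravity and scaled component motion.
# The event graph has no connected logic, and the blueprint declares no variables of its own.
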AnimBlueprint /BRCosmetics/Accessories/FORT_Backpacks/Backpack_M_MED_Giggle/Meshes/M_MED_Giggle_Pack_AnimBP
target_skeleton=&/M_MED_Giggle_Pack_Skeleton
graphs [
 AnimGraph {nodes [
   n1 = "Transform (Modify) Bone - Bone: pack"(ComponentPose=Local To Component(LocalPose=Local Space Ref Pose()), Alpha=0.75, Translation="0.000,3,0.000", Rotation="0,0,10") {
    rotationMode=BMM_Additive
    translationMode=BMM_Additive
    translationSpace=BCS_BoneSpace
    alphaBoolBlend {blendOption=Linear}
    boneToModify {boneName=pack}
    rotation {roll=10}
    translation {y=3}
   }
   n2 = Get bIsSkyDiving()
   n3 = RigidBody(
    ComponentPose="Transform (Modify) Bone - Bone: pack"(ComponentPose=@n1, Alpha="To Float (Boolean)"(InBool=@n2), Translation="0,0,0", Rotation="0,0,20") {rotationMode=BMM_Additive, alphaBoolBlend {blendOption=Linear}, boneToModify {boneName=pack}, rotation {roll=20}}
    Alpha=0.9
   ) {
    bOverrideWorldGravity=true
    overridePhysicsAsset=&/M_MED_Giggle_Pack_Physics1
    alphaBoolBlend {blendOption=Linear}
    componentAppliedLinearAccClamp=<3000 3000 3000>
    componentLinearAccScale=<0.2 0.2 0.4>
    componentLinearVelScale=<0.1 0.1 0.2>
    overrideWorldGravity {z=-980}
   }
   Output Pose(Result=Component To Local(ComponentPose=@n3)) {layerGroup=None}
  ]}
 EventGraph {nodes [
   TryGetPawnOwner()
   Event BlueprintUpdateAnimation()
  ]}
]
variables []
